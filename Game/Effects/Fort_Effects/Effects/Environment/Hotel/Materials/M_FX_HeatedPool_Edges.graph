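# Translucent, lit Niagara material for heated-pool edge steam in the Hotel environment effects.
# Vertex blue blends a Fresnel-faded, UV-mapped smoke layer, pushed outward in XY by noise, with a world-XY-projected smoke layer.
# Both layers distort smoke by offsetting panner time with a second sample and take particle color and alpha.
# Opacity adds high-quality depth fade, a U-edge fade and a dynamic-parameter world-space cutout whose center reuses MaskX for both axes.
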Material /Game/Effects/Fort_Effects/Effects/Environment/Hotel/Materials/M_FX_HeatedPool_Edges
graph {
 properties {
  bApplyCloudFogging=true
  bUseMaterialAttributes=true
  bUsedWithNiagaraMeshParticles=true
  bUsedWithNiagaraSprites=true
  blendMode=BLEND_Translucent
  pixelDepthOffsetMode=PDOM_Legacy
  translucencyLightingMode=TLM_VolumetricPerVertexNonDirectional
  usedShadingModels=Default Lit
 }
 nodes [
  ComponentMask(None=TextureCoordinate()) {g=true}
  ComponentMask(None=ActorPositionWS()) {b=true}
  Constant3Vector() {constant=<rgba 0 0 1 0>}
  Constant() {r=1}
  Fresnel()
  OneMinus(None=VertexColor().G)
  n1 = TextureCoordinate()
  n2 = SphereMask(A=Fmod(A=ComponentMask(None=@n1) {r=true}, B=Constant() {r=1}), B=Constant() {r=0.5}) {attenuationRadius=0.5, hardnessPercent=70}
  OneMinus(None=@n2)
  PixelNormalWS()
  n3 = ParticleColor()
  n4 = DynamicParameter() {defaultValue=<rgba 512 1 1 1>, paramNames [WorldScale,EdgeHeight,Param3,Param4]}
  n5 = Multiply(A=TextureCoordinate() {uTiling=2, vTiling=0.5}) {constB=0.5}
  n6 = Multiply(A=Time()) {constB=0.075}
  n7 = Multiply(A=TextureSampleParameter2D(UVs=Panner(Coordinate=@n5, Time=@n6) {speedX=0.002, speedY=0.125}) {parameterName=T_SmokeTexture, texture=/Game/Global/Textures/Default/Tileable_Patterns/T_Smoke_Puff_Tile_Up}.G) {constB=2}
  n8 = TextureSampleParameter2D(UVs=Panner(Coordinate=@n5, Time=Add(A=@n6, B=@n7)) {speedX=0.01, speedY=0.1}) {parameterName=T_SmokeTexture, texture=&T_Smoke_Puff_Tile_Up}
  n9 = VertexColor()
  n10 = Multiply(
   A=MaterialFunctionCall(Exponent=@n4.Param3) {materialFunction=&AxisAlignedFresnel}.Fresnel
   B=Multiply(A=Multiply(A=@n3.A, B=@n8.G), B=Power(Base=@n9.R))
  )
  n11 = OneMinus(None=VertexColor().B)
  n12 = VertexNormalWS()
  n13 = Constant3Vector() {constant=<rgba 1 1 0 0>}
  n14 = Panner() {speedY=0.1}
  n15 = TextureSample(UVs=@n14) {texture=&T_Smooth_tile}
  n16 = TextureCoordinate()
  n17 = Multiply(A=Multiply(A=Multiply(A=@n12, B=@n13), B=Multiply(A=@n15.R) {constB=128}), B=OneMinus(None=ComponentMask(None=@n16) {g=true}))
  n18 = ParticleColor()
  n19 = VertexColor()
  n20 = Divide(A=WorldPosition().XY, B=@n4.WorldScale)
  n21 = Multiply(A=Time()) {constB=0.1}
  n22 = Panner(Coordinate=@n20, Time=@n21) {speedY=-0.05}
  n23 = Add(
   A=@n21
   B=Multiply(A=TextureSampleParameter2D(UVs=@n22) {parameterName=T_SmokeTexture, texture=&T_Smoke_Puff_Tile_Up}.G) {constB=2}
  )
  n24 = Multiply(
   A=@n18.A
   B=TextureSampleParameter2D(UVs=Panner(Coordinate=@n20, Time=@n23) {speedX=0.01, speedY=-0.1}) {parameterName=T_SmokeTexture, texture=&T_Smoke_Puff_Tile_Up}.G
  )
  n25 = BlendMaterialAttributes(
   A=MakeMaterialAttributes!(
    BaseColor=@n3.RGB
    Metallic
    Specular
    Roughness
    Anisotropy
    EmissiveColor
    Opacity=Multiply(A=@n10, B=@n11)
    OpacityMask
    Normal
    Tangent
    WorldPositionOffset=Multiply(A=@n11, B=@n17)
    SubsurfaceColor
    ClearCoat
    ClearCoatRoughness
    AmbientOcclusion
    Refraction
    CustomizedUVs_0
    CustomizedUVs_1
    CustomizedUVs_2
    CustomizedUVs_3
    CustomizedUVs_4
    CustomizedUVs_5
    CustomizedUVs_6
    CustomizedUVs_7
    PixelDepthOffset
    ShadingModel
    Displacement
   )
   B=MakeMaterialAttributes(BaseColor=@n18.RGB, Opacity=Multiply(A=@n19.B, B=@n24))
   Alpha=VertexColor().B
  )
  n26 = BreakMaterialAttributes!(Attr=@n25)
  n27 = DepthFade(Opacity=@n26.Opacity)
  n28 = QualitySwitch(Default=@n26.Opacity, High=@n27, Epic=@n27)
  n29 = DynamicParameter() {parameterIndex=1, defaultValue=<rgba 512 1 1 1>, paramNames [WorldScale/MaskX, EdgeHeight/MaskY, Param3/MaskR, Param4]}
  n30 = OneMinus(None=SphereMask(A=WorldPosition().XY, B=AppendVector(A=@n29.WorldScale/MaskX, B=@n29.WorldScale/MaskX), Radius=@n29.Param3/MaskR) {hardnessPercent=10})
  SetMaterialAttributes!(MaterialAttributes=@n25, Opacity=Multiply(A=Multiply(A=@n28, B=@n30), B=@n2)) {attributeSetTypes [#1]}
  TextureSampleParameter2D(UVs=@n14) {parameterName=T_SmokeTexture, texture=&T_Smoke_Puff_Tile_Up}
  TextureSample() {texture=&T_Inky_Smoke_Tile}
 ]
}
