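# Niagara system for ambient dust motes in the Hotel environment, using one GPU sprite emitter that continuously fills a sphere volume.
# Motes drift on curl noise, drag and fading gravity, fading over life with a waveform flicker.
# Camera distance drives sprite size, and time of day tints color at spawn, though its spawn-rate and alpha curves are flat.
# The DEBUG emitter and indoor-player spawn rate are disabled.
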NiagaraSystem /Game/Effects/Fort_Effects/Effects/Environment/Hotel/NS_Env_Hotel_Motes_Volume
dependencies {
 usedDataInterfaces [&NiagaraDataInterfaceCurve, &NiagaraDataInterfaceCamera, &NiagaraDataInterfaceColorCurve]
 usedDynamicInputs [&SimulationPosition]
 usedModules [
  &SystemState
  &self:SystemSpawnScript.SystemScriptSource.SystemScriptGraph.NiagaraNodeAssignment_0.SetVariables_#1
 ]
}
emitters {
 DEBUG {
  bEnabled=false
  emitterUpdateScript [
   &EmitterState()
   &SpawnRate(SpawnRate=500, Use Spawn Probability=false)
  ]
  particleSpawnScript [
   /SetVariables(Particles.RandSpreadCurve:NiagaraFloat=&FloatFromCurve(
     FloatCurve=DataInterface() {propertyValues {Curve {keys [
         {interpMode=RCIM_Cubic, tangentMode=RCTM_User, time=0, value=0.7}
         {interpMode=RCIM_Cubic, tangentMode=RCTM_User, time=0.8, value=0.2}
         {arriveTangent=2.7919, interpMode=RCIM_Cubic, leaveTangent=2.7919, tangentMode=RCTM_User, time=1, value=1}
        ]}}}
     CurveIndex=&RandomRangeFloat(Minimum=0, Maximum=1)
    )) {isSetParameters=true}
   &InitializeParticle(Color=<rgba 0.9565 0 1 1>, Lifetime=5, Sprite Size Mode=::NewEnumerator3, UsePositionOffset=false, Write Material Random=true)
   &ShapeLocation(
    Cone Length:NiagaraFloat=&Lerp_Float(A=200, B=1500, Alpha=$Particles.RandSpreadCurve)
    Cone Mode:ENiagara_ConeMode=::NewEnumerator1
    Cone Surface Distribution=0.5
    MaskSpawnGroup=false
    Override Seed=false
    Rotation Mode=::NewEnumerator1
    Shape Primitive=::NewEnumerator5
    UseOwnerScale=true
    Wedge Horizontal Angle:NiagaraFloat=&Lerp_Float(A=30, B=75, Alpha=$Particles.RandSpreadCurve)
    Wedge Vertical Angle:NiagaraFloat=&Lerp_Float(A=25, B=50, Alpha=$Particles.RandSpreadCurve)
   )
   &ShapeLocation(
    Cone Angle:NiagaraFloat=&Lerp_Float(A=10, B=75, Alpha=$Particles.RandSpreadCurve)
    Cone Inner Angle:NiagaraFloat=&Lerp_Float(A=10, B=10, Alpha=$Particles.RandSpreadCurve)
    Cone Length:NiagaraFloat=&Lerp_Float(A=200, B=1500, Alpha=$Particles.RandSpreadCurve)
    MaskSpawnGroup=false
    Override Seed=false
    Rotation Axis=<0 0 -1>
    Rotation Mode=::NewEnumerator1
    Shape Primitive=::NewEnumerator5
    UseOwnerScale=true
   ) {enabled=false, module=ShapeLocation001}
  ]
  particleUpdateScript [
   /SetVariables(Particles.DEBUG:NiagaraFloat=&VectorLength(Vector=Linked() {linkedVariable {name=Particles.Position, type=NiagaraPosition}})) {isSetParameters=true}
   &ParticleState()
   &ScaleColor(
    Scale Alpha=&Lerp_Float(A=1, B=0, Alpha=$Particles.RandSpreadCurve)
    Scale RGB=&VectorFromFloat(Value=&Lerp_Float(A=1, B=0, Alpha=$Particles.RandSpreadCurve))
    ScaleA=true
    ScaleRGB=true
   )
   &ScaleColor(
    Scale Alpha=&Lerp_Float(
     A=1
     B=0
     Alpha=&RemapRange(
      Input Value To Remap=&VectorLength(Vector=Linked() {linkedVariable {name=Particles.Position, type=NiagaraPosition}})
      Input Min=0
      Input Max=1200
      Output Min=0
      Output Max=1
     )
    )
    Scale RGB=&VectorFromFloat(Value=&Lerp_Float(
      A=1
      B=0
      Alpha=&RemapRange(
       Input Value To Remap=&VectorLength(Vector=Linked() {linkedVariable {name=Particles.Position, type=NiagaraPosition}})
       Input Min=0
       Input Max=1200
       Output Min=0
       Output Max=1
      )
     ))
    ScaleA=true
    ScaleRGB=true
   ) {enabled=false, module=ScaleColor001}
  ]
  rendererClasses [&NiagaraSpriteRendererProperties]
  renderers [{rendererClass=&NiagaraSpriteRendererProperties}]
 }
 Dust_Motes {
  simTarget=GPUComputeSim
  emitterUpdateScript [
   /SetVariables(Emitter.CameraPostion:Vector3f=&ReturnCameraVector(Camera Property=::NewEnumerator0, Coordinate Space=::Simulation)) {isSetParameters=true}
   &EmitterState(
    Life Cycle Mode=::NewEnumerator1
    Loop Duration=&RandomRangeFloat(Minimum=5, Maximum=10)
    Recalculate Duration Each Loop=true
    UseLoopDelay=false
   )
   &SpawnRate(SpawnRate=&Lerp_Float(A=0, B=100, Alpha=$NPC.FortniteNPC.IsPlayerIndoors), Use Spawn Probability=false) {enabled=false}
   &SpawnRate(
    SpawnRate=&Multiply_Float(
     A=1000
     B=&FloatFromCurve(
      FloatCurve=DataInterface() {propertyValues {Curve {keys [<at 0 1>,<at 0.24 1>,<at 0.43 1>]}}}
      CurveIndex=Linked() {linkedVariable {name=NPC.FortniteNPC.TimeOfDay_0-1, type=NiagaraFloat}}
     )
    )
    Use Spawn Probability=false
   ) {module=SpawnRate001}
  ]
  particleSpawnScript [
   /SetVariables(Particles.RandSpreadCurve:NiagaraFloat=&FloatFromCurve(
     FloatCurve=DataInterface() {propertyValues {Curve {keys [
         {interpMode=RCIM_Cubic, tangentMode=RCTM_User, time=0, value=0.7}
         {interpMode=RCIM_Cubic, tangentMode=RCTM_User, time=0.8, value=0.4}
         {arriveTangent=2.7919, interpMode=RCIM_Cubic, leaveTangent=2.7919, tangentMode=RCTM_User, time=1, value=1}
        ]}}}
     CurveIndex=&RandomRangeFloat(Minimum=0, Maximum=1)
    )) {isSetParameters=true}
   &InitializeParticle(
    Lifetime=&UniformRangedFloat(Minimum=3, Maximum=5)
    Mass Mode=::NewEnumerator1
    Sprite Rotation Mode=::NewEnumerator2
    Sprite Size=&Vector2DFromFloat(Value=&RandomRangeFloat(Minimum=0.25, Maximum=0.5))
    Sprite Size Mode=::NewEnumerator4
    UsePositionOffset=false
    Write Material Random=true
   ) {module=InitializeParticle001}
   &ShapeLocation(MaskSpawnGroup=false, Override Seed=false, Rotation Mode=::NewEnumerator1, Sphere Radius=800, UseOwnerScale=true)
   &AddVelocity(Velocity=&RandomVector(Vector Scale=&RandomRangeFloat(Minimum=0.5, Maximum=1)))
   /Niagara/Modules/Update/Camera/MaintainInCameraParticleScale()
   &DynamicMaterialParameters(
    Index 0 Param 1=2
    Index 0 Param 2=50
    Index 0 Param 3=200
    Module.Param0WriteEnabled=true
    Module.Param1WriteEnabled=true
    Module.Param2WriteEnabled=true
    Module.Param3WriteEnabled=true
   ) {module=DynamicMaterialParameters001}
   &Color(Color=&ColorFromCurve(
     ColorCurve=DataInterface() {propertyValues {
       AlphaCurve {keys [<at 0 1 RCIM_Cubic>]}
       BlueCurve {keys [<at 0 1 RCIM_Cubic>,<at 0.2015 0.2768>,<at 0.7445 0.2768>,<at 1 1 RCIM_Cubic>]}
       GreenCurve {keys [<at 0 0.7227 RCIM_Cubic>,<at 0.2015 0.777>,<at 0.7445 0.777>,<at 1 0.7227 RCIM_Cubic>]}
       RedCurve {keys [<at 0 0.5 RCIM_Cubic>,<at 0.2015 1>,<at 0.7445 1>,<at 1 0.5 RCIM_Cubic>]}
      }}
     CurveIndex=Linked() {linkedVariable {name=NPC.FortniteNPC.TimeOfDay_0-1, type=NiagaraFloat}}
    ))
  ]
  particleUpdateScript [
   &ParticleState()
   &self:Camera_to_Particle_Distance_Calculation(Camera Postion:Vector3f=$Emitter.CameraPostion, Distance=150, Particle Position:NiagaraPosition=$Particles.Position) {module=Camera_to_Particle_Distance_Calculation}
   &ScaleSpriteSize(Scale Factor=&Vector2DFromFloat(Value=&FloatFromCurve(
      FloatCurve=DataInterface() {propertyValues {Curve {keys [<at 0 0.25 RCIM_Cubic>,<at 1 1.5 RCIM_Cubic>]}}}
      CurveIndex=$Particles.NormalizedDistance
     )))
   &Drag(Drag=0.5, Use Linear Drag=true, Use Rotational Drag=false)
   &GravityForce(Gravity=&MakeVector(
     X=0
     Y=0
     Z=&FloatFromCurve(
      FloatCurve=DataInterface() {propertyValues {Curve {keys [<at 0 1 RCIM_Cubic>,<at 1 0>]}}}
      Scale Curve=&RandomRangeFloat(Minimum=-1, Maximum=-3)
     )
    ))
   &CurlNoiseForce(
    Curl Noise Cone Mask Angle=120
    Mask Curl Noise=true
    Noise Frequency=&RandomRangeFloat(Minimum=200, Maximum=400)
    Noise Quality / Cost=::NewEnumerator0
    Noise Strength=&RandomRangeFloat(Minimum=10, Maximum=25)
    Pan Noise=true
    Pan Noise Field=<0 0 1>
   )
   &SpriteRotationRate(Rotation Rate=&RandomRangeFloat(Minimum=-45, Maximum=45))
   &ScaleColor(
    Scale Alpha=&FloatFromCurve(
     FloatCurve=DataInterface() {propertyValues {Curve {keys [
         {interpMode=RCIM_Cubic, tangentMode=RCTM_User, time=0, value=0}
         {arriveTangent=1.8182, leaveTangent=1.8182, time=0.35, value=1}
         <at 0.55 1>
         <at 1 0>
        ]}}}
     Scale Curve=&FloatFromCurve(
      FloatCurve=DataInterface() {propertyValues {Curve {keys [<at 0 1 RCIM_Cubic>,<at 0.22 1>,<at 0.43 1>]}}}
      CurveIndex=Linked() {linkedVariable {name=NPC.FortniteNPC.TimeOfDay_0-1, type=NiagaraFloat}}
      Scale Curve=2
     )
    )
    Scale RGB=&Multiply_VectorByFloat(Vector=<1 1 1>, Float=1)
    ScaleA=true
    ScaleRGB=true
   )
   &ScaleColor(
    Scale Alpha=&Multiply_Float(
     A=&Lerp_Float(A=1, B=0.3, Alpha=$Particles.RandSpreadCurve)
     B=&Waveform(
      "Waveform [1]"=::NewEnumerator3
      "[1] Frequency"=&RandomRangeFloat(Minimum=0.2, Maximum=0.4)
      "[1] Amplitude Min/Max"=<0 1>
     )
    )
    Scale RGB=&VectorFromFloat(Value=1)
    ScaleA=true
    ScaleRGB=true
   ) {module=ScaleColor001}
   &SolveForcesAndVelocity(Clamp Velocity=false, Limit Acceleration=false)
   &DynamicMaterialParameters(
    Index 0 Param 1=1
    Index 0 Param 2=150
    Index 0 Param 3=150
    Index 1 Param 1=1
    Module.Index0Param0Write=true
    Module.Index0Param1Write=true
    Module.Index0Param2Write=true
    Module.Index0Param3Write=true
    Module.Index1Param0Write=true
    Module.Index1Param1Write=true
    Module.Index1Param2Write=false
    Module.Index1Param3Write=true
    Write Parameter Index 1=true
   )
  ]
  rendererClasses [&NiagaraSpriteRendererProperties]
  renderers [{rendererClass=&NiagaraSpriteRendererProperties}]
 }
}
summary {
 systemName=NS_Env_Hotel_Motes_Volume
 emitters [
  {bEnabled=false, emitterName=DEBUG, rendererClasses [&NiagaraSpriteRendererProperties]}
  {emitterName=Dust_Motes, simTarget=GPUComputeSim, rendererClasses [&NiagaraSpriteRendererProperties]}
 ]
 userVariables []
}
