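# Niagara system for the Mountain Hotel heated pool, drawing surface layers and soft particles drifting over the water.
# SurfaceLayer bursts persistent sprite and mesh particles placed and sized from a user float4 array, passing a mask position and radius to materials.
# SurfaceParticles fills a box sized from array entry 0 with two spawn groups, split by visibility tag, that drift along Y, fade and animate material parameters through curves.
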NiagaraSystem /Game/Effects/Fort_Effects/Effects/Environment/Hotel/NS_MountainHotel_HeatedPool
dependencies {
 usedDataInterfaces [&NiagaraDataInterfaceSpriteRendererInfo, &NiagaraDataInterfaceCurve]
 usedDynamicInputs []
 usedModules [&SystemState]
}
emitters {
 SurfaceLayer {
  emitterUpdateScript [
   &EmitterState(Life Cycle Mode=::NewEnumerator1, Loop Behavior=::NewEnumerator1, Loop Duration Mode=::NewEnumerator1, UseLoopDelay=false)
   &SpawnBurst_Instantaneous(
    Spawn Count=&self:ScratchDynamicInput_01(Vec4Array:NiagaraDataInterfaceArrayFloat4=$User.Plane_Pos_Size_Array)
    Use Spawn Probability=false
    UseLoopCountLimit=false
   )
  ]
  particleSpawnScript [
   &InitializeParticle(Color=<rgba 1 1 1 0.4>, Lifetime=2, Sprite Size Mode=::NewEnumerator4, UsePositionOffset=false, Write Material Random=true)
   /SetVariables(Particles.PoolData:Vector4f=&self:ScratchDynamicInput(
     Index:NiagaraInt32=&ReturnParticleIDAttributes(Particle ID Attribute=::NewEnumerator1)
     Pooldata_in:NiagaraDataInterfaceArrayFloat4=$User.Plane_Pos_Size_Array
    )) {isSetParameters=true}
   /SetVariables(
    Particles.Position=&MakeVector(
     X=&MakeFloatFromVector4(Vector 4=$Particles.PoolData, Channel=::NewEnumerator0)
     Y=&MakeFloatFromVector4(Vector 4=$Particles.PoolData, Channel=::NewEnumerator1)
     Z=10
    )
    Particles.SpriteSize=&MakeVector2D(
     X=&MakeFloatFromVector4(Vector 4=$Particles.PoolData, Channel=::NewEnumerator2)
     Y=&MakeFloatFromVector4(Vector 4=$Particles.PoolData, Channel=::NewEnumerator3)
    )
   ) {isSetParameters=true}
   /SetVariables(Particles.Scale=&Multiply_Vector(
     A=&Multiply_VectorByFloat(
      Vector=&MakeVector(
       X=&Multiply_Float(A=&MakeFloatFromVector2d(Vector 2D=$Particles.SpriteSize), B=1)
       Y=&MakeFloatFromVector2d(Vector 2D=$Particles.SpriteSize, Channel=::NewEnumerator1)
       Z=250
      )
      Float=0.01
     )
     B=<1 1 1>
    )) {isSetParameters=true}
  ]
  particleUpdateScript [
   &ParticleState(
    Kill Particles When Lifetime Has Elapsed=false
    Let Infinitely Lived Particles Die When Emitter Deactivates=true
    Loop Particles Lifetime=true
   )
   &ScaleColor(ScaleA=true, ScaleRGB=false)
   &SolveForcesAndVelocity(Clamp Velocity=false, Limit Acceleration=false)
   &SpriteFacingAndAlignment(Sprite Facing=<0 0 1>, WriteAlignment=true, WriteFacing=true)
   &DynamicMaterialParameters(
    Index 0 Param 1=8000
    Index 0 Param 2=200
    Index 0 Param 3=1
    Index 1 Param 1=&MakeFloatFromVector(Vector=$User.MaskOutPos, Channel=::NewEnumerator0)
    Index 1 Param 2=&MakeFloatFromVector(Vector=$User.MaskOutPos, Channel=::NewEnumerator1)
    Index 1 Param 3=$User.MaskR
    Module.Index0Param0Write=true
    Module.Index0Param1Write=true
    Module.Index0Param2Write=true
    Module.Index0Param3Write=true
    Module.Index1Param0Write=true
    Module.Index1Param1Write=true
    Module.Index1Param2Write=true
    Module.Index1Param3Write=true
    Write Parameter Index 1=true
   )
   &CameraOffset(Camera Offset Amount=20)
  ]
  rendererClasses [&NiagaraSpriteRendererProperties, &NiagaraMeshRendererProperties]
  renderers [
   {rendererClass=&NiagaraSpriteRendererProperties}
   {rendererClass=&NiagaraMeshRendererProperties, rendererIndex=1}
  ]
 }
 SurfaceParticles {
  emitterUpdateScript [
   &EmitterState()
   &SpawnRate(SpawnGroup=1, Use Spawn Probability=false) {module=SpawnRate001}
   &SpawnRate(SpawnRate=2, Use Spawn Probability=false)
  ]
  particleSpawnScript [
   &InitializeParticle(
    Color=&MaskLinearColorBySpawnGroup(Masked Linear Color=<rgba 1 1 1 0.025>, Passthrough Linear Color=<rgba 1 1 1 0.175>)
    Lifetime=&MaskFloatBySpawnGroup(
     Masked Float=&RandomRangeFloat(Minimum=5, Maximum=7)
     Passthrough Float=&RandomRangeFloat(Minimum=2, Maximum=4)
    )
    UsePositionOffset=false
    Write Material Random=true
   )
   /SetVariables(Particles.PoolData:Vector4f=&self:ScratchDynamicInput(Index:NiagaraInt32=0, Pooldata_in:NiagaraDataInterfaceArrayFloat4=$User.Plane_Pos_Size_Array)) {isSetParameters=true}
   &ShapeLocation(
    Box Size=&MakeVector(
     X=&MakeFloatFromVector4(Vector 4=$Particles.PoolData, Channel=::NewEnumerator2)
     Y=&MakeFloatFromVector4(Vector 4=$Particles.PoolData, Channel=::NewEnumerator3)
     Z=1
    )
    MaskSpawnGroup=false
    Override Seed=false
    Shape Primitive=::NewEnumerator2
    Surface Only=false
    UseOwnerScale=true
   )
   /SetVariables(
    Particles.SpriteRotation:NiagaraFloat=&MaskFloatBySpawnGroup(Masked Float=&RandomRangeFloat(Minimum=0, Maximum=350), Passthrough Float=-90)
    Particles.SpriteSize=&MaskVector2dBySpawnGroup(Masked Vector2D=<500 500>, Passthrough Vector2D=<600 600>, Spawn Group Mask=0)
    Particles.VisibilityTag=$Particles.SpawnGroup
   ) {isSetParameters=true}
  ]
  particleUpdateScript [
   &ParticleState()
   &SubUVAnimation(
    Output Alternate Renderer Binding=true
    "Sprite Renderer "=DataInterface() {propertyValues {SpriteRenderer=&self:SurfaceParticles_0.NiagaraSpriteRendererProperties_0}}
    UseEndFrame=false
    UseStartFrame=false
   )
   &AccelerationForce(Acceleration=&Multiply_VectorByFloat(Vector=<0 10 0>, Float=&RandomRangeFloat(Minimum=0.25, Maximum=1.15)))
   &Drag(Drag=&RandomRangeFloat(Minimum=1, Maximum=4), Use Linear Drag=true, Use Rotational Drag=true)
   &DynamicMaterialParameters(
    Index 0 Param 1=&Add_Float(
     A=0
     B=&FloatFromCurve(
      FloatCurve=DataInterface() {propertyValues {Curve {keys [
          {arriveTangent=2.7533, interpMode=RCIM_Cubic, leaveTangent=1.8958, tangentMode=RCTM_User, time=0, value=1}
          {arriveTangent=0.8292, interpMode=RCIM_Cubic, leaveTangent=0.8292, tangentMode=RCTM_User, time=0.321, value=1.3453}
          {arriveTangent=0.0529, interpMode=RCIM_Cubic, leaveTangent=0.2021, tangentMode=RCTM_User, time=1, value=1.5}
         ]}}}
      CurveIndex=&Add_Float(A=&Multiply_Float(A=$Particles.Initial.MaterialRandom, B=0.5), B=$Particles.NormalizedAge)
     )
    )
    Index 0 Param 2=&FloatFromCurve(FloatCurve=DataInterface() {propertyValues {Curve {keys [
         {arriveTangent=0.0453, interpMode=RCIM_Cubic, tangentMode=RCTM_Break, time=0, value=-1}
         {arriveTangent=0.0604, interpMode=RCIM_Cubic, leaveTangent=0.0604, tangentMode=RCTM_User, time=1, value=-1}
        ]}}})
    Index 0 Param 3=&FloatFromCurve(
     FloatCurve=DataInterface() {propertyValues {Curve {keys [
         {arriveTangent=-53.9199, interpMode=RCIM_Cubic, leaveTangent=-8.9727, tangentMode=RCTM_User, time=0, value=1}
         {arriveTangent=0.2347, interpMode=RCIM_Cubic, leaveTangent=0.2723, tangentMode=RCTM_Break, time=0.25, value=0}
        ]}}}
     Scale Curve=0
    )
    Index 0 Param 4=&FloatFromCurve(FloatCurve=DataInterface() {propertyValues {Curve {keys [<at 0.5 1>,<at 1 0>]}}})
    Index 2 Param 1=&FloatFromCurve(
     FloatCurve=DataInterface() {propertyValues {Curve {keys [
         {interpMode=RCIM_Cubic, leaveTangent=4.2485, tangentMode=RCTM_Break, time=0, value=0}
         {arriveTangent=1.0548, interpMode=RCIM_Cubic, leaveTangent=1.0548, tangentMode=RCTM_User, time=0.17, value=0.5715}
         <at 1 1>
        ]}}}
     Scale Curve=3
    )
    Index 2 Param 2=&FloatFromCurve(
     FloatCurve=DataInterface() {propertyValues {Curve {keys [
         {arriveTangent=0.9316, interpMode=RCIM_Cubic, leaveTangent=0.9316, tangentMode=RCTM_User, time=0, value=0.1}
         {arriveTangent=0.0259, interpMode=RCIM_Cubic, leaveTangent=0.0259, tangentMode=RCTM_User, time=1, value=0.4}
        ]}}}
     Scale Curve=&RandomRangeFloat(Minimum=0.1, Maximum=0.35)
    )
    Index 2 Param 3=&FloatFromCurve(FloatCurve=DataInterface() {propertyValues {Curve {keys [
         {interpMode=RCIM_Cubic, leaveTangent=2.3064, tangentMode=RCTM_Break, time=0, value=1}
         {interpMode=RCIM_Cubic, leaveTangent=11.1233, tangentMode=RCTM_Break, time=1, value=2}
        ]}}})
    Index 2 Param 4=&FloatFromCurve(
     FloatCurve=DataInterface() {propertyValues {Curve {keys [
         {arriveTangent=-0.6867, interpMode=RCIM_Cubic, leaveTangent=-0.6867, tangentMode=RCTM_User, time=0, value=1.25}
         {interpMode=RCIM_Cubic, leaveTangent=-1.2956, tangentMode=RCTM_Break, time=1, value=0.75}
        ]}}}
     Scale Curve=2
    )
    Module.Param0WriteEnabled=true
    Module.Param0WriteEnabled2=true
    Module.Param1WriteEnabled=false
    Module.Param1WriteEnabled2=true
    Module.Param2WriteEnabled=true
    Module.Param2WriteEnabled2=true
    Module.Param3WriteEnabled=true
    Module.Param3WriteEnabled2=true
    Write Parameter Index 2=true
   )
   &SolveForcesAndVelocity(Clamp Velocity=false, Limit Acceleration=false)
   &ScaleColor(
    Scale Alpha=&FloatFromCurve(FloatCurve=DataInterface() {propertyValues {Curve {keys [
         {interpMode=RCIM_Cubic, tangentMode=RCTM_User, time=0, value=0}
         {interpMode=RCIM_Cubic, tangentMode=RCTM_User, time=0.5, value=1}
         {tangentMode=RCTM_User, time=1, value=0}
        ]}}})
    ScaleA=true
    ScaleRGB=true
   )
  ]
  rendererClasses [&NiagaraSpriteRendererProperties]
  renderers [
   {rendererClass=&NiagaraSpriteRendererProperties}
   {rendererClass=&NiagaraSpriteRendererProperties, rendererIndex=1}
  ]
 }
}
summary {
 systemName=NS_MountainHotel_HeatedPool
 emitters [
  {emitterName=SurfaceLayer, rendererClasses [&NiagaraSpriteRendererProperties, &NiagaraMeshRendererProperties]}
  {emitterName=SurfaceParticles, rendererClasses [&NiagaraSpriteRendererProperties]}
 ]
 userVariables [
  User.MaskOutPos {type=Vector3f, description="", defaultValue {struct=Vector3f, value=<0 0 0>}}
  User.MaskR {type=NiagaraFloat, defaultValue=0, description=""}
  User.Plane_Pos_Size_Array {
   type=NiagaraDataInterfaceArrayFloat4
   description=""
   defaultValue {
    struct=NiagaraExt_VariableValue_DataInterface
    value {
     dataInterface=&self:NiagaraDataInterfaceArrayFloat4_0
     dataInterfaceClass=/Script/Niagara.NiagaraDataInterfaceArrayFloat4
    }
   }
  }
 ]
}
